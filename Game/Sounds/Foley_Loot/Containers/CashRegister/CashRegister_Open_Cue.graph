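# Sound cue for the foley of a cash register door opening, filed under loot container sounds.
# A single wave player feeds the root directly, with no randomization or modulation, at a boosted volume multiplier.
# It references the item interact attenuation asset and the loot reward sound class.
# Its concurrency override uses a prevent-new rule, so playing instances are kept over new ones when the limit is reached.
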SoundCue /Game/Sounds/Foley_Loot/Containers/CashRegister/CashRegister_Open_Cue
root=WavePlayer_1
nodes [{id=WavePlayer_1, properties {soundWaveAssetPtr=&/Cash_Cash_Register_Door_Open}}]
properties {
 attenuationSettings=/Game/Sounds/Attenuation/Fort_Item_Interact_Att
 bOverrideConcurrency=true
 maxDistance=1850
 soundClassObject=&Fort_Loot_Reward_PSM
 volumeMultiplier=3
 attenuationOverrides {falloffDistance=1000, attenuationShapeExtents {x=500}}
 concurrencyOverrides {resolutionRule=PreventNew}
}
