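# Looping sound cue for the foley heard during a long search of a cash register loot container.
# A single looping wave player feeds the output directly, with no randomization or modulation.
# It plays in the loot collect sound class with the mono item-interact attenuation and a slightly reduced volume.
# A concurrency override prevents new instances from starting once the limit is reached, rather than cutting off ones already playing.
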SoundCue /Game/Sounds/Foley_Loot/Containers/CashRegister/CashRegister_Searching_Long_Loop_Cue
root=WavePlayer_0
nodes [{
  id=WavePlayer_0
  properties {bLooping=true, soundWaveAssetPtr=&/Cash_Register_Searching_Long_Loop_01}
 }]
properties {
 attenuationSettings=/Game/Sounds/Attenuation/Fort_Item_Interact_Mono_Att
 bOverrideConcurrency=true
 maxDistance=1500
 soundClassObject=&Fort_Loot_Collect
 volumeMultiplier=0.8141
 attenuationOverrides {falloffDistance=1000, attenuationShapeExtents {x=500}}
 concurrencyOverrides {resolutionRule=PreventNew}
}
