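# Sound cue for the short cash register searching foley, part of the loot container sounds.
# A single Wave Player feeds the root directly with one recorded wave, with no randomization or modulation.
# It uses the loot collect sound class and the shared item interact attenuation asset, and overrides concurrency with a rule that prevents new instances once the limit is reached.
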SoundCue /Game/Sounds/Foley_Loot/Containers/CashRegister/CashRegister_Searching_Short_Cue
root=WavePlayer_0
nodes [{id=WavePlayer_0, properties {soundWaveAssetPtr=&/Cash_Register_Searching_Short_01}}]
properties {
 attenuationSettings=/Game/Sounds/Attenuation/Fort_Item_Interact_Att
 bOverrideConcurrency=true
 maxDistance=1850
 soundClassObject=&Fort_Loot_Collect
 volumeMultiplier=1
 attenuationOverrides {falloffDistance=1000, attenuationShapeExtents {x=500}}
 concurrencyOverrides {resolutionRule=PreventNew}
}
